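# Sound cue for the close sound of the Season 11 umbrella glider cosmetic in Athena.
# A single wave player feeds the umbrella close wave straight to the output, with no randomization or modulation.
# It uses shared glider open/close attenuation and concurrency, a glider-close audio property sheet and the glider landing sound class, with slightly reduced volume.
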SoundCue /BRCosmetics/Sounds/Gliders/Season_11_Umbrella/Athena_Season_11_Umbrella_Close_Cue
root=WavePlayer_1
nodes [{id=WavePlayer_1, properties {soundWaveAssetPtr=&/Athena_Umbrella_Season11_Close}}]
properties {
 attenuationSettings=/Game/Sounds/Attenuation/Fort_Player_Glider_OpenClose_Athena_Att
 audioPropertiesSheet=/BRCosmetics/Sounds/Settings/PropertySheets/Glider/APS_Cosmetic_Glider_Close
 maxDistance=10000
 soundClassObject=&Fort_Gliders_Land
 volumeMultiplier=0.7
 concurrencySet [&Athena_GliderOpenClose_Conc]
}
