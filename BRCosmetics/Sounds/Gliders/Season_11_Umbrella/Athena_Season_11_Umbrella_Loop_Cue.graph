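# Sound cue for the Season 11 umbrella glider's looping sound in Battle Royale.
# A single looping wave player plays the umbrella loop wave as the cue's root, with no mixing, randomization or modulation.
# Spatialization and routing come from a glider thrust attenuation asset, a cosmetic glider loop audio properties sheet and the gliders sound class.
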SoundCue /BRCosmetics/Sounds/Gliders/Season_11_Umbrella/Athena_Season_11_Umbrella_Loop_Cue
root=WavePlayer_3
nodes [{id=WavePlayer_3, properties {bLooping=true, soundWaveAssetPtr=&/Athena_Umbrella_Season11_Loop}}]
properties {
 attenuationSettings=&Fort_Player_Glider_Thrust_Athena_Att
 audioPropertiesSheet=&APS_Cosmetic_Glider_Loop
 maxDistance=12500
 soundClassObject=&Fort_Gliders_Land
 volumeMultiplier=1
}
