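# Sound cue for the opening sound of the Season 11 Storm Umbrella glider cosmetic.
# A single wave player feeds a modulator that slightly randomizes pitch on each play.
# It uses the shared glider open/close attenuation and concurrency, the glider land sound class, and the cosmetic glider-open audio property sheet.
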SoundCue /BRCosmetics/Sounds/Gliders/Season_11_Umbrella/Glider_Umbrella_Storm_Open_Cue
root=Modulator_0
nodes [
 {id=Modulator_0, children [WavePlayer_1], properties {pitchMax=1.02, pitchMin=0.98}}
 {id=WavePlayer_1, properties {soundWaveAssetPtr=&/Glider_Umbrella_Storm_Open_01}}
]
properties {
 attenuationSettings=/Game/Sounds/Attenuation/Fort_Player_Glider_OpenClose_Athena_Att
 audioPropertiesSheet=/BRCosmetics/Sounds/Settings/PropertySheets/Glider/APS_Cosmetic_Glider_Open
 maxDistance=10000
 soundClassObject=&Fort_Gliders_Land
 volumeMultiplier=1
 concurrencySet [&Athena_GliderOpenClose_Conc]
}
